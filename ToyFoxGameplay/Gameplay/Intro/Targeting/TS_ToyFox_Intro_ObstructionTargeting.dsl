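; Targeting task for obstruction targeting in the ToyFox intro that overrides where the query originates.
; GetSourceLocation casts the targeting context's source to the giant player pawn and returns its mesh's pelvis socket location.
; Using the socket rather than the actor location accounts for the character animating during the query.
; The parent implementation serves as fallback when that path does not return.
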(fn GetSourceLocation
  (params (TargetingHandle :type Targeting%20Request%20Handle%20Structure%20%28by%20ref%29))
  (bind _returnvalue (Targeting|GetTargetingSubsystem TargetingHandle))
  ; GetSourceLocation: Overridden to return the location of the socket instead of actor cause this is when the character is animating.
  (Utilities|FlowControl|Sequence
    (:then_0
      (Utilities|IsValid _returnvalue
        (:Is%20Valid
          (bind _astoy_fox_giant_player_pawn (Utilities|Casting|CastToToyFoxGiantPlayerPawn (Utilities|Struct|BreakTargetingSourceContext (TargetingSystem|TargetingTypes|GetTargetingSourceContext _returnvalue TargetingHandle))))
          (bind _mesh (Variables|/Script/Engine.Character|GetMesh _astoy_fox_giant_player_pawn)
            (:then
              (return (Transformation|GetSocketLocation _mesh "pelvis")))
            (:else)))))
    (:then_1
      (bind _returnvalue_2 (|CallParentFunction TargetingHandle))
      (return _returnvalue_2))))
